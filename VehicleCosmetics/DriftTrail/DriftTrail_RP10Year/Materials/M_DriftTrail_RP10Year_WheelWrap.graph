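# Masked, unlit Niagara material for the wheel-wrap effect of the RP10Year drift trail.
# A gradient texture, tinted by face side and multiplied by a fake world-environment reflection, blends from a dark desaturated tint through the EmissiveCtrl dynamic parameter.
# A wheel-wrap mask function cuts the shape, optionally multiplied by a panning ribbon texture.
# Sine/cosine vertex offset ripples along U, driven by time and particle color, with a switchable second offset.
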Material /VehicleCosmetics/DriftTrail/DriftTrail_RP10Year/Materials/M_DriftTrail_RP10Year_WheelWrap
graph {
 properties {
  bAutomaticallySetUsageInEditor=false
  bUsedWithNiagaraMeshParticles=true
  bUsedWithNiagaraRibbons=true
  bUsedWithNiagaraSprites=true
  blendMode=BLEND_Masked
  shadingModel=MSM_Unlit
  usedShadingModels=Unlit
 }
 nodes [
  n1 = TextureCoordinate() {uTiling=0.14}
  n2 = VectorParameter() {parameterName=WheelwrapScale, defaultValue=<rgba 1.5 0.5 0 1>}
  n3 = TextureSampleParameter2D(UVs=Panner(Coordinate=MaterialFunctionCall(UVs=@n1, Texture Scale=@n2.RGB) {materialFunction=&ScaleUVsByCenter}.UVs)) {parameterName=SwitchTexture, texture=/VehicleCosmetics/DriftTrail/DriftTrail_RP10Year/Textures/T_VCVFX_Gradient_WheelWrap}
  n4 = Multiply(A=Desaturation(None=@n3.RGB), B=VectorParameter() {parameterName=DarkColor2, defaultValue=<rgba 0.221 0.0596 0.014 1>}.RGB)
  n5 = Saturate(None=LinearInterpolate(A=VectorParameter() {parameterName=DarkColor, defaultValue=<rgba 0.221 0.0596 0.014 1>}.RGB, B=@n3.RGB, Alpha=TwoSidedSign()))
  n6 = DynamicParameter() {paramNames [EmissiveCtrl,Param2,Param3,Param4]}
  LinearInterpolate(A=@n4, B=@n5, Alpha=@n6.EmissiveCtrl)
  n7 = Constant()
  n8 = MaterialFunctionCall!(Desaturate=ScalarParameter() {defaultValue=0.5, parameterName=Reflection_Saturation}, "Tint (Curve)"=@n7, Tint Amount=@n7, Cubemap) {materialFunction=/DealershipSharedLibrary/FX/Materials/MaterialFunctions/VC_MF_WorldEnv}
  n9 = ScalarParameter() {defaultValue=25, parameterName=Metalness}
  n10 = Saturate(None=ConstantBiasScale(None=Desaturation(None=@n8.WorldSpace)) {bias=-0.75, scale=2})
  n11 = Power(Base=Multiply(A=Multiply(A=@n8.WorldSpace, B=@n9), B=ConstantBiasScale(None=@n10) {bias=0.25, scale=1.75}), Exp=Constant() {r=1})
  MaterialFunctionCall(Emissive=LinearInterpolate(A=@n4, B=Multiply(A=@n5, B=@n11), Alpha=@n6.EmissiveCtrl)) {materialFunction=&MF_FX_EmissiveAdjustments}
  n12 = Add(A=Constant() {r=1}, B=@n3.A)
  n13 = TextureCoordinate()
  n14 = VectorParameter() {parameterName=WheelwrapScale, defaultValue=<rgba 1.5 0.5 0 1>}
  n15 = TextureSample(UVs=Panner(Coordinate=MaterialFunctionCall(UVs=@n13, Texture Scale=@n14.RGB) {materialFunction=&ScaleUVsByCenter}.UVs) {speedX=-0.5}) {texture=/DealershipSharedLibrary/FX/Textures/Gradient_Masks/T_VCVFX_Ribbon01_A}
  Multiply(
   A=StaticSwitchParameter(True=Multiply(A=@n12, B=@n15.R), False=Constant() {r=1}) {parameterName=AlphaActive}
   B=MaterialFunctionCall!(Multiply=Constant() {r=1}, Power EXP=Constant() {r=10}) {materialFunction=/DealershipSharedLibrary/FX/Materials/MaterialFunctions/MF_VCVFX_WheelwrapMask}.Result
  )
  n16 = Constant()
  n17 = ScalarParameter() {defaultValue=-0.5, parameterName=2WPO_Offset}
  n18 = Constant()
  n19 = ComponentMask(None=Add(A=TextureCoordinate(), B=AppendVector(A=@n17, B=@n18))) {r=true}
  n20 = OneMinus(None=Ceil(None=@n19))
  n21 = Multiply(A=@n19, B=Multiply(A=Saturate(None=@n20), B=ScalarParameter() {defaultValue=1, parameterName=2WPOIntensity}))
  n22 = ScalarParameter() {defaultValue=-0.5, parameterName=X_Offset}
  n23 = Constant()
  n24 = ComponentMask(None=Add(A=TextureCoordinate(), B=AppendVector(A=@n22, B=@n23))) {r=true}
  n25 = ParticleColor()
  n26 = Divide(A=@n24, B=@n25.A)
  n27 = Multiply(A=OneMinus(None=Time()), B=ScalarParameter() {defaultValue=1, group=WPO, parameterName=Speed})
  n28 = Sine(None=Add(A=@n26, B=@n27))
  n29 = Cosine(None=Add(A=@n27, B=@n26))
  n30 = Multiply(
   A=@n25.RGB
   B=Multiply(
    A=MaterialFunctionCall(X=@n28, Y=@n29, Z=@n28) {materialFunction=&MakeFloat3}.Result
    B=ScalarParameter() {defaultValue=1, group=WPO, parameterName=Intensity}
   )
  )
  n31 = Multiply(A=@n24, B=ConstantBiasScale(None=Transform(None=@n30) {transformSourceType=TRANSFORMSOURCE_Local}))
  n32 = Add(A=AppendVector(A=AppendVector(A=@n16, B=@n21), B=@n16), B=@n31)
  Multiply(A=StaticSwitchParameter(True=Transform(None=@n32) {transformSourceType=TRANSFORMSOURCE_Local}, False=@n31) {parameterName=2WPO})
 ]
}
