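# Translucent unlit Niagara material for the wheel-wrap piece of the RP10Year drift trail cosmetic.
# Emissive blends tinted gradient with two-sided color multiplied by world-environment reflections, faking metal without lighting.
# Switchable opacity combines gradient alpha with a panning ribbon mask, optionally sine-modulated through GOLDRIBBON and IsFront.
# World position offset adds time-driven sine/cosine waves weighted along U and shaped by particle color, plus an optional second bend via the 2WPO switch.
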Material /VehicleCosmetics/DriftTrail/DriftTrail_RP10Year/Materials/M_DriftTrail_RP10Year_WheelWrap_V2
graph {
 properties {
  bAutomaticallySetUsageInEditor=false
  bUsedWithNiagaraMeshParticles=true
  bUsedWithNiagaraRibbons=true
  bUsedWithNiagaraSprites=true
  blendMode=BLEND_Translucent
  shadingModel=MSM_Unlit
  usedShadingModels=Unlit
 }
 nodes [
  n1 = LinearInterpolate(A=TextureCoordinate() {uTiling=0}, B=TextureCoordinate() {unMirrorU=true}, Alpha=ScalarParameter() {parameterName=GOLDRIBBON})
  n2 = TextureSampleParameter2D(UVs=Panner(Coordinate=OneMinus(None=@n1))) {parameterName=Switch Texture, texture=/VehicleCosmetics/DriftTrail/DriftTrail_RP10Year/Textures/T_VCVFX_Gradient_WheelWrap}
  n3 = Desaturation(None=@n2.A)
  n4 = VectorParameter() {parameterName=DarkColor2, defaultValue=<rgba 0.221 0.0596 0.014 1>}
  n5 = Saturate(None=LinearInterpolate(A=VectorParameter() {parameterName=DarkColor, defaultValue=<rgba 0.221 0.0596 0.014 1>}.RGB, B=@n2.RGB, Alpha=TwoSidedSign()))
  n6 = Constant()
  n7 = MaterialFunctionCall!(Desaturate=ScalarParameter() {defaultValue=0.5, parameterName=Reflection_Saturation}, "Tint (Curve)"=@n6, Tint Amount=@n6, Cubemap) {materialFunction=/DealershipSharedLibrary/FX/Materials/MaterialFunctions/VC_MF_WorldEnv}
  n8 = ScalarParameter() {defaultValue=25, parameterName=Metalness}
  n9 = Saturate(None=ConstantBiasScale(None=Desaturation(None=@n7.WorldSpace)) {bias=-0.75, scale=2})
  n10 = Power(Base=Multiply(A=Multiply(A=@n7.WorldSpace, B=@n8), B=ConstantBiasScale(None=@n9) {bias=0.25, scale=1.75}), Exp=Constant() {r=0.75})
  MaterialFunctionCall(Emissive=LinearInterpolate(A=Multiply(A=@n3, B=@n4.RGB), B=Multiply(A=@n5, B=@n10), Alpha=ScalarParameter() {defaultValue=1, parameterName=Emissive})) {materialFunction=&MF_FX_EmissiveAdjustments}
  n11 = Panner(Coordinate=MaterialFunctionCall(UVs=TextureCoordinate(), Texture Scale=VectorParameter() {parameterName=WheelwrapScale, defaultValue=<rgba 0.85 0.5 0 1>}.RGB) {materialFunction=&ScaleUVsByCenter}.UVs) {speedX=-0.5}
  n12 = Multiply(
   A=@n2.A
   B=SmoothStep(Value=TextureSample(UVs=@n11) {samplerSource=SSM_Clamp_WorldGroupSettings, texture=/DealershipSharedLibrary/FX/Textures/Gradient_Masks/T_VCVFX_Ribbon01_A}.RGB) {constMin=0.5}
  )
  n13 = ComponentMask(None=TextureCoordinate() {uTiling=0.8}) {r=true}
  n14 = LinearInterpolate(A=@n12, B=Multiply(A=Sine(None=@n13) {period=0.3}, B=@n12), Alpha=ScalarParameter() {parameterName=GOLDRIBBON})
  n15 = ScalarParameter() {parameterName=IsFront}
  Saturate(None=StaticSwitchParameter(True=LinearInterpolate(A=@n14, B=@n12, Alpha=@n15), False=Constant() {r=1}) {parameterName=AlphaActive})
  n16 = Constant()
  n17 = ScalarParameter() {defaultValue=-0.5, parameterName=2WPO_Offset}
  n18 = Constant()
  n19 = ComponentMask(None=Add(A=TextureCoordinate(), B=AppendVector(A=@n17, B=@n18))) {r=true}
  n20 = Ceil(None=@n19)
  n21 = Multiply(A=Saturate(None=OneMinus(None=@n20)), B=ScalarParameter() {defaultValue=8, parameterName=2WPOIntensity})
  n22 = AppendVector(A=AppendVector(A=@n16, B=Multiply(A=@n19, B=@n21)), B=@n16)
  n23 = ScalarParameter() {defaultValue=-0.5, parameterName=X_Offset}
  n24 = Constant()
  n25 = ComponentMask(None=Add(A=TextureCoordinate(), B=AppendVector(A=@n23, B=@n24))) {r=true}
  n26 = ParticleColor()
  n27 = Divide(A=@n25, B=@n26.A)
  n28 = Multiply(A=OneMinus(None=Time()), B=ScalarParameter() {defaultValue=1, group=WPO, parameterName=Speed})
  n29 = Sine(None=Add(A=@n27, B=@n28))
  n30 = Cosine(None=Add(A=@n28, B=@n27))
  n31 = Multiply(
   A=@n26.RGB
   B=Multiply(
    A=MaterialFunctionCall(X=@n29, Y=@n30, Z=@n29) {materialFunction=&MakeFloat3}.Result
    B=ScalarParameter() {defaultValue=1, group=WPO, parameterName=Intensity}
   )
  )
  n32 = Multiply(A=@n25, B=ConstantBiasScale(None=Transform(None=@n31) {transformSourceType=TRANSFORMSOURCE_Local}))
  StaticSwitchParameter(True=Transform(None=Add(A=@n22, B=@n32)) {transformSourceType=TRANSFORMSOURCE_Local}, False=@n32) {parameterName=2WPO}
 ]
}
