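# Masked, two-sided Niagara material for the RP10Year drift trail ribbon, also usable on sprites and mesh particles.
# A tinted gradient texture blends with a fake environment-cubemap reflection via the EmissiveCtrl dynamic parameter, then passes through a shared emissive adjustment function.
# Time-driven sine/cosine offsets, shaped by U, particle color and dynamic parameters, wave the ribbon, with a switchable second offset.
# A separate branch combines panning gold, square and noise textures, faded at both U ends.
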Material /VehicleCosmetics/DriftTrail/DriftTrail_RP10Year/Materials/M_DriftTrail_RP10yearRibbonMat
graph {
 properties {
  bAutomaticallySetUsageInEditor=false
  bUsedWithNiagaraMeshParticles=true
  bUsedWithNiagaraRibbons=true
  bUsedWithNiagaraSprites=true
  blendMode=BLEND_Masked
  twoSided=true
  usedShadingModels=Default Lit
 }
 nodes [
  n1 = VectorParameter() {parameterName=Ribbon_Coord_Strip, defaultValue=<rgba 0.68 1 0 0>}
  n2 = TextureSampleParameter2D(UVs=Multiply(A=TextureCoordinate(), B=AppendVector(A=@n1.R, B=@n1.R))) {parameterName=Switch Texture, texture=/VehicleCosmetics/DriftTrail/DriftTrail_RP10Year/Textures/T_VCVFX_Gradient}
  n3 = Multiply(A=Desaturation(None=@n2.RGB), B=VectorParameter() {parameterName=DarkColor2, defaultValue=<rgba 0.221 0.0596 0.014 1>}.RGB)
  MaterialExpressionConstantClamp_1 = Saturate(None=LinearInterpolate(A=VectorParameter() {parameterName=DarkColor, defaultValue=<rgba 0.221 0.0596 0.014 1>}.RGB, B=@n2.RGB, Alpha=TwoSidedSign()))
  n4 = DynamicParameter() {defaultValue=<rgba 1 7 -1.5 0>, paramNames [EmissiveCtrl, Intensity, Speed, Custom X Offset]}
  LinearInterpolate(A=@n3, B=@MaterialExpressionConstantClamp_1, Alpha=@n4.EmissiveCtrl)
  n5 = Constant()
  n6 = MaterialFunctionCall!(Desaturate=ScalarParameter() {defaultValue=0.5, parameterName=Reflection_Saturation}, "Tint (Curve)"=@n5, Tint Amount=@n5, Cubemap) {materialFunction=/DealershipSharedLibrary/FX/Materials/MaterialFunctions/VC_MF_WorldEnv}
  n7 = Multiply(A=@n6.WorldSpace, B=ScalarParameter() {defaultValue=25, parameterName=Metalness})
  MaterialExpressionConstantClamp_2 = Saturate(None=ConstantBiasScale(None=Desaturation(None=@n6.WorldSpace)) {bias=-0.75, scale=2})
  n8 = ConstantBiasScale(None=@MaterialExpressionConstantClamp_2) {bias=0.25, scale=1.75}
  n9 = Multiply(A=@MaterialExpressionConstantClamp_1, B=Power(Base=Multiply(A=@n7, B=@n8), Exp=Constant() {r=0.75}))
  MaterialFunctionCall(Emissive=LinearInterpolate(
    A=LinearInterpolate(A=@n3, B=@n9, Alpha=@n4.EmissiveCtrl)
    B=VectorParameter() {parameterName=EmissiveColor, defaultValue=<rgba 1 0.7484 0 1>}.RGB
    Alpha=ScalarParameter() {parameterName=UseEmissiveVector}
   )) {materialFunction=&MF_FX_EmissiveAdjustments}
  n10 = Constant()
  n11 = ScalarParameter() {defaultValue=-0.5, parameterName=2WPO_Offset}
  n12 = Constant()
  n13 = ComponentMask(None=Add(A=TextureCoordinate(), B=AppendVector(A=@n11, B=@n12))) {r=true}
  MaterialExpressionConstantClamp_0 = Saturate(None=OneMinus(None=Ceil(None=@n13)))
  n14 = Multiply(A=@n13, B=Multiply(A=@MaterialExpressionConstantClamp_0, B=ScalarParameter() {defaultValue=8, parameterName=2WPOIntensity}))
  n15 = ScalarParameter() {defaultValue=-0.5, parameterName=X_Offset}
  n16 = Constant()
  n17 = ComponentMask(None=Add(A=TextureCoordinate(), B=AppendVector(A=@n15, B=@n16))) {r=true}
  MaterialExpressionVertexColor_0 = ParticleColor()
  n18 = Divide(A=@n17, B=@MaterialExpressionVertexColor_0.A)
  n19 = Time()
  n20 = Multiply(A=OneMinus(None=Add(A=@n4.Custom X Offset, B=@n19)), B=@n4.Speed)
  n21 = Sine(None=Add(A=@n18, B=@n20))
  n22 = Cosine(None=Add(A=@n20, B=@n18))
  n23 = Multiply(
   A=@MaterialExpressionVertexColor_0.RGB
   B=Multiply(
    A=MaterialFunctionCall(X=@n21, Y=@n22, Z=@n21) {materialFunction=&MakeFloat3}.Result
    B=@n4.Intensity
   )
  )
  n24 = Multiply(A=@n17, B=ConstantBiasScale(None=Transform(None=@n23) {transformSourceType=TRANSFORMSOURCE_Local}))
  n25 = Add(A=AppendVector(A=AppendVector(A=@n10, B=@n14), B=@n10), B=@n24)
  Multiply(A=StaticSwitchParameter(True=Transform(None=@n25) {transformSourceType=TRANSFORMSOURCE_Local}, False=@n24) {parameterName=2WPO})
  n26 = AppendVector(A=ScalarParameter() {defaultValue=1, parameterName=S19_XTilingAmount}, B=Constant() {r=1})
  n27 = Panner(Coordinate=Multiply(A=TextureCoordinate() {uTiling=2}, B=@n26)) {speedX=-0.5}
  n28 = TextureCoordinate() {uTiling=1.5}
  n29 = TextureSample(UVs=Panner(Coordinate=Multiply(A=@n28, B=@n26)) {speedX=-1}) {texture=/DealershipSharedLibrary/FX/Textures/Gradient_Masks/T_VCFVX_Squares}
  n30 = Multiply(A=TextureCoordinate() {uTiling=0.3, vTiling=0.3}, B=@n26)
  n31 = OneMinus(None=TextureSample(UVs=Panner(Coordinate=@n30) {speedX=-0.2, speedY=0.1}) {texture=/DealershipSharedLibrary/FX/Textures/Gradient_Masks/T_VCVFX_Blob_Noise_A}.RGB)
  n32 = Multiply(
   A=Constant3Vector() {constant=<rgba 0.8 0.4 0.01 0>}
   B=Subtract(
    A=TextureSample(UVs=@n27) {texture=/DealershipSharedLibrary/FX/Textures/Gradient_Masks/T_VCFVX_S19ChampionsRoad_Golden}.RGB
    B=Multiply(A=@n29.RGB, B=@n31)
   )
  )
  n33 = ComponentMask(None=TextureCoordinate()) {r=true}
  n34 = OneMinus(None=@n33)
  n35 = Multiply(A=@n34, B=@n34)
  n36 = Saturate(None=Multiply(A=@n33) {constB=3})
  Saturate(None=Multiply(A=Multiply(A=@n32, B=@n4.EmissiveCtrl), B=Multiply(A=@n35, B=@n36)))
 ]
}
